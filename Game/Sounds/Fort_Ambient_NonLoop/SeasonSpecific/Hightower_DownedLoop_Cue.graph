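# Outdoor ambient sound cue for the season-specific Hightower "downed" loop, layering a steady bed with intermittent random hits.
# A mixer blends the looping bed wave with a louder branch that repeatedly picks one of six equally weighted ambient hits, pitch-modulated and randomly delayed.
# An attenuation override sets distance falloff with low-pass filtering and a wide stereo spread, routed to the Athena outdoor ambient sound class.
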SoundCue /Game/Sounds/Fort_Ambient_NonLoop/SeasonSpecific/Hightower_DownedLoop_Cue
root=Mixer_0
nodes [
 {id=Delay_0, children [Modulator_0], properties {delayMax=2}}
 {id=Looping_0, children [Delay_0]}
 {id=Mixer_0, children [Looping_0,WavePlayer_0], properties {inputVolume [1.2,0.6762]}}
 {id=Modulator_0, children [Random_0], properties {pitchMin=0.8}}
 {id=Random_0, children [WavePlayer_1,WavePlayer_2,WavePlayer_3,WavePlayer_4,WavePlayer_5,WavePlayer_6], properties {weights [1,1,1,1,1,1]}}
 {id=WavePlayer_0, properties {bLooping=true, soundWaveAssetPtr=&/Hightower_DownedLoop}}
 {id=WavePlayer_1, properties {soundWaveAssetPtr=&/Hightower_AmbientHit_01}}
 {id=WavePlayer_2, properties {soundWaveAssetPtr=&/Hightower_AmbientHit_02}}
 {id=WavePlayer_3, properties {soundWaveAssetPtr=&/Hightower_AmbientHit_03}}
 {id=WavePlayer_4, properties {soundWaveAssetPtr=&/Hightower_AmbientHit_04}}
 {id=WavePlayer_5, properties {soundWaveAssetPtr=&/Hightower_AmbientHit_05}}
 {id=WavePlayer_6, properties {soundWaveAssetPtr=&/Hightower_AmbientHit_06}}
]
properties {
 bOverrideAttenuation=true
 maxDistance=3749.1951
 soundClassObject=&Fort_Ambient_Outdoors_Athena
 volumeMultiplier=0.752
 attenuationOverrides {
  bAttenuateWithLPF=true
  falloffDistance=2249.1951
  lPFFrequencyAtMin=386.6391
  lPFRadiusMax=1227.2043
  lPFRadiusMin=1242.1268
  stereoSpread=5000
  attenuationShapeExtents {x=1500}
 }
}
